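# Physics asset for the jacket of the male medium PeaceScarf body cosmetic, named as the crouch variant.
# Seven four-bone capsule chains (back left/center/right, both sides, both fronts) hang from small base anchor bodies through swing-limited constraints.
# Twist is locked except on the top two right-front joints, and back chains swing more tightly than front ones.
# Pelvis, lower spine, thigh and calf capsules have no constraints and give body shapes for collision.
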PhysicsAsset /BRCosmetics/Characters/Player/Male/Medium/Bodies/M_MED_PeaceScarf/Meshes/M_MED_PeaceScarf_Physics_Crouch
bodies {
 base_jacket_bk_c [{shapeType=Capsule, length=0.505, radius=0.505, shapeName=base_jacket_bk_c_capsule, rotation=<rot 0 -90 0>}]
 base_jacket_bk_l [{shapeType=Capsule, length=0.505, radius=0.505, shapeName=base_jacket_bk_l_capsule, rotation=<rot 0 -90 0>}]
 base_jacket_bk_r [{shapeType=Capsule, length=0.505, radius=0.505, shapeName=base_jacket_bk_r_capsule, rotation=<rot 0 -90 0>}]
 base_jacket_fr_l [{shapeType=Capsule, length=0.505, radius=0.505, shapeName=base_jacket_fr_l_capsule, rotation=<rot 0 -90 0>}]
 base_jacket_fr_r [{shapeType=Capsule, length=0.505, radius=0.505, shapeName=base_jacket_fr_r_capsule, rotation=<rot 0 -90 0>}]
 base_jacket_sd_l [{shapeType=Capsule, length=0.505, radius=0.505, shapeName=base_jacket_sd_l_capsule, rotation=<rot 0 -90 0>}]
 base_jacket_sd_r [{shapeType=Capsule, length=0.505, radius=0.505, shapeName=base_jacket_sd_r_capsule, rotation=<rot 0 -90 0>}]
 calf_l [{
   shapeType=Capsule
   length=35.1809
   radius=7.6868
   shapeName=calf_l_capsule
   center=<-26.2684 -0.9415 1.2103>
   rotation=<rot 87.3059 179.9983 179.9989>
  }]
 calf_r [{
   shapeType=Capsule
   length=35.1809
   radius=7.6868
   shapeName=calf_r_capsule
   center=<26.2684 0.9415 -1.2103>
   rotation=<rot 87.3059 179.9983 -0.0011>
  }]
 dyn_jacket_bk_01_c [{shapeType=Capsule, length=12, radius=8, shapeName=dyn_jacket_bk_01_c_capsule, center=<8.6056 0 6.9873>, rotation=<rot 90 -90 0>}]
 dyn_jacket_bk_01_l [{shapeType=Capsule, length=12, radius=8, shapeName=dyn_jacket_bk_01_l_capsule, center=<8.6056 0 6.9873>, rotation=<rot 90 -90 0>}]
 dyn_jacket_bk_01_r [{shapeType=Capsule, length=12, radius=8, shapeName=dyn_jacket_bk_01_r_capsule, center=<8.6056 0 6.9873>, rotation=<rot 90 -90 0>}]
 dyn_jacket_bk_02_c [{shapeType=Capsule, length=12, radius=8, shapeName=dyn_jacket_bk_02_c_capsule, center=<8.6056 0 6.9873>, rotation=<rot 90 -90 0>}]
 dyn_jacket_bk_02_l [{shapeType=Capsule, length=12, radius=8, shapeName=dyn_jacket_bk_02_l_capsule, center=<8.6056 0 6.9873>, rotation=<rot 90 -90 0>}]
 dyn_jacket_bk_02_r [{shapeType=Capsule, length=12, radius=8, shapeName=dyn_jacket_bk_02_r_capsule, center=<8.6056 0 6.9873>, rotation=<rot 90 -90 0>}]
 dyn_jacket_bk_03_c [{shapeType=Capsule, length=12, radius=8, shapeName=dyn_jacket_bk_03_c_capsule, center=<8.6056 0 6.9873>, rotation=<rot 90 -90 0>}]
 dyn_jacket_bk_03_l [{shapeType=Capsule, length=12, radius=8, shapeName=dyn_jacket_bk_03_l_capsule, center=<8.6056 0 6.9873>, rotation=<rot 90 -90 0>}]
 dyn_jacket_bk_03_r [{shapeType=Capsule, length=12, radius=8, shapeName=dyn_jacket_bk_03_r_capsule, center=<8.6056 0 6.9873>, rotation=<rot 90 -90 0>}]
 dyn_jacket_bk_04_c [{shapeType=Capsule, length=12, radius=8, shapeName=dyn_jacket_bk_04_c_capsule, center=<8.6056 0 6.9873>, rotation=<rot 90 -90 0>}]
 dyn_jacket_bk_04_l [{shapeType=Capsule, length=12, radius=8, shapeName=dyn_jacket_bk_04_l_capsule, center=<8.6056 0 6.9873>, rotation=<rot 90 -90 0>}]
 dyn_jacket_bk_04_r [{shapeType=Capsule, length=12, radius=8, shapeName=dyn_jacket_bk_04_r_capsule, center=<8.6056 0 6.9873>, rotation=<rot 90 -90 0>}]
 dyn_jacket_fr_01_l [{shapeType=Capsule, length=8, radius=6, shapeName=dyn_jacket_fr_01_l_capsule, center=<8.4863 -0.3269 5.4551>, rotation=<rot 0 0 76.6239>}]
 dyn_jacket_fr_01_r [{
   shapeType=Capsule
   length=8
   radius=6
   shapeName=dyn_jacket_fr_01_r_capsule
   center=<8.4863 0.2005 4.9235>
   rotation=<rot 0.0001 -0.0001 -76.6239>
  }]
 dyn_jacket_fr_02_l [{
   shapeType=Capsule
   length=9.282
   radius=7
   shapeName=dyn_jacket_fr_02_l_capsule
   center=<8.2721 -1.8324 5.6204>
   rotation=<rot 78.8379 -89.9531 -0.0003>
  }]
 dyn_jacket_fr_02_r [{
   shapeType=Capsule
   length=9.282
   radius=6
   shapeName=dyn_jacket_fr_02_r_capsule
   center=<8.272 1.6889 4.8202>
   rotation=<rot 89.5121 -90.0128 179.9845>
  }]
 dyn_jacket_fr_03_l [{shapeType=Capsule, length=12, radius=6, shapeName=dyn_jacket_fr_03_l_capsule, center=<8.6056 0 5.836>, rotation=<rot 90 -90 0>}]
 dyn_jacket_fr_03_r [{shapeType=Capsule, length=12, radius=6, shapeName=dyn_jacket_fr_03_r_capsule, center=<8.6056 0 5.836>, rotation=<rot 90 90 0>}]
 dyn_jacket_fr_04_l [{shapeType=Capsule, length=12, radius=6, shapeName=dyn_jacket_fr_04_l_capsule, center=<8.6056 0 5.836>, rotation=<rot 90 -90 0>}]
 dyn_jacket_fr_04_r [{shapeType=Capsule, length=12, radius=6, shapeName=dyn_jacket_fr_04_r_capsule, center=<8.6056 0 5.836>, rotation=<rot 90 90 0>}]
 dyn_jacket_sd_01_l [{shapeType=Capsule, length=12, radius=8, shapeName=dyn_jacket_sd_01_l_capsule, center=<8.6056 0 6.9873>, rotation=<rot 90 -90 0>}]
 dyn_jacket_sd_01_r [{shapeType=Capsule, length=12, radius=8, shapeName=dyn_jacket_sd_01_r_capsule, center=<8.6056 0 6.9873>, rotation=<rot 90 -90 0>}]
 dyn_jacket_sd_02_l [{shapeType=Capsule, length=12, radius=8, shapeName=dyn_jacket_sd_02_l_capsule, center=<8.6056 0 4.9907>, rotation=<rot 90 -90 0>}]
 dyn_jacket_sd_02_r [{shapeType=Capsule, length=12, radius=8, shapeName=dyn_jacket_sd_02_r_capsule, center=<8.6056 0 6.9873>, rotation=<rot 90 -90 0>}]
 dyn_jacket_sd_03_l [{shapeType=Capsule, length=12, radius=8, shapeName=dyn_jacket_sd_03_l_capsule, center=<8.6056 0 3.7749>, rotation=<rot 90 -90 0>}]
 dyn_jacket_sd_03_r [{shapeType=Capsule, length=12, radius=8, shapeName=dyn_jacket_sd_03_r_capsule, center=<8.6056 0 6.9873>, rotation=<rot 90 -90 0>}]
 dyn_jacket_sd_04_l [{shapeType=Capsule, length=12, radius=8, shapeName=dyn_jacket_sd_04_l_capsule, center=<8.6056 0 6.9873>, rotation=<rot 90 -90 0>}]
 dyn_jacket_sd_04_r [{shapeType=Capsule, length=12, radius=8, shapeName=dyn_jacket_sd_04_r_capsule, center=<8.6056 0 6.9873>, rotation=<rot 90 -90 0>}]
 pelvis [{shapeType=Capsule, length=10.2603, radius=8.0704, shapeName=pelvis_capsule, center=<0 -1.0694 0>}]
 spine_01 [{shapeType=Capsule, length=10.6469, radius=7.2325, shapeName=spine_01_capsule, center=<0 -0.6262 0>, rotation=<rot 0 -90 0>}]
 spine_02 [{shapeType=Capsule, length=7.8853, radius=7.7879, shapeName=spine_02_capsule, center=<0 -0.1467 0>}]
 thigh_l [{shapeType=Capsule, length=43.6009, radius=6.7908, shapeName=thigh_l_capsule, center=<-19.5126 0 0.7533>, rotation=<rot 90 0 0>}]
 thigh_r [{shapeType=Capsule, length=42.8103, radius=7.5814, shapeName=thigh_r_capsule, center=<19.5126 -1.4242 -0.7533>, rotation=<rot 90 180 0>}]
}
constraints [
 {
  bone1Name=dyn_jacket_bk_01_c
  bone2Name=base_jacket_bk_c
  swing1LimitDegrees=3
  swing2LimitDegrees=15
  twistLimitDegrees=45
  twistMotion=Locked
 }
 {
  bone1Name=dyn_jacket_bk_02_c
  bone2Name=dyn_jacket_bk_01_c
  swing1LimitDegrees=3
  swing2LimitDegrees=15
  twistLimitDegrees=5
  twistMotion=Locked
 }
 {
  bone1Name=dyn_jacket_bk_03_c
  bone2Name=dyn_jacket_bk_02_c
  swing1LimitDegrees=3
  swing2LimitDegrees=15
  twistLimitDegrees=5
  twistMotion=Locked
 }
 {
  bone1Name=dyn_jacket_bk_04_c
  bone2Name=dyn_jacket_bk_03_c
  swing1LimitDegrees=3
  swing2LimitDegrees=15
  twistLimitDegrees=5
  twistMotion=Locked
 }
 {
  bone1Name=dyn_jacket_bk_01_l
  bone2Name=base_jacket_bk_l
  swing1LimitDegrees=3
  swing2LimitDegrees=15
  twistLimitDegrees=45
  twistMotion=Locked
 }
 {
  bone1Name=dyn_jacket_bk_02_l
  bone2Name=dyn_jacket_bk_01_l
  swing1LimitDegrees=3
  swing2LimitDegrees=15
  twistLimitDegrees=5
  twistMotion=Locked
 }
 {
  bone1Name=dyn_jacket_bk_03_l
  bone2Name=dyn_jacket_bk_02_l
  swing1LimitDegrees=3
  swing2LimitDegrees=15
  twistLimitDegrees=5
  twistMotion=Locked
 }
 {
  bone1Name=dyn_jacket_bk_04_l
  bone2Name=dyn_jacket_bk_03_l
  swing1LimitDegrees=3
  swing2LimitDegrees=15
  twistLimitDegrees=5
  twistMotion=Locked
 }
 {
  bone1Name=dyn_jacket_bk_01_r
  bone2Name=base_jacket_bk_r
  swing1LimitDegrees=3
  swing2LimitDegrees=15
  twistLimitDegrees=45
  twistMotion=Locked
 }
 {
  bone1Name=dyn_jacket_bk_02_r
  bone2Name=dyn_jacket_bk_01_r
  swing1LimitDegrees=3
  swing2LimitDegrees=15
  twistLimitDegrees=5
  twistMotion=Locked
 }
 {
  bone1Name=dyn_jacket_bk_03_r
  bone2Name=dyn_jacket_bk_02_r
  swing1LimitDegrees=3
  swing2LimitDegrees=15
  twistLimitDegrees=5
  twistMotion=Locked
 }
 {
  bone1Name=dyn_jacket_bk_04_r
  bone2Name=dyn_jacket_bk_03_r
  swing1LimitDegrees=3
  swing2LimitDegrees=15
  twistLimitDegrees=5
  twistMotion=Locked
 }
 {
  bone1Name=dyn_jacket_fr_02_l
  bone2Name=dyn_jacket_fr_01_l
  swing1LimitDegrees=20
  swing2LimitDegrees=15
  twistLimitDegrees=5
  twistMotion=Locked
 }
 {
  bone1Name=dyn_jacket_fr_03_l
  bone2Name=dyn_jacket_fr_02_l
  swing1LimitDegrees=30
  swing2LimitDegrees=30
  twistLimitDegrees=5
  twistMotion=Locked
 }
 {
  bone1Name=dyn_jacket_fr_04_l
  bone2Name=dyn_jacket_fr_03_l
  swing1LimitDegrees=30
  swing2LimitDegrees=30
  twistLimitDegrees=5
  twistMotion=Locked
 }
 {bone1Name=dyn_jacket_fr_02_r, bone2Name=dyn_jacket_fr_01_r, swing1LimitDegrees=20, swing2LimitDegrees=20, twistLimitDegrees=5}
 {
  bone1Name=dyn_jacket_fr_03_r
  bone2Name=dyn_jacket_fr_02_r
  swing1LimitDegrees=20
  swing2LimitDegrees=30
  twistLimitDegrees=5
  twistMotion=Locked
 }
 {
  bone1Name=dyn_jacket_fr_04_r
  bone2Name=dyn_jacket_fr_03_r
  swing1LimitDegrees=20
  swing2LimitDegrees=30
  twistLimitDegrees=5
  twistMotion=Locked
 }
 {
  bone1Name=dyn_jacket_sd_01_l
  bone2Name=base_jacket_sd_l
  swing1LimitDegrees=15
  swing2LimitDegrees=15
  twistLimitDegrees=45
  twistMotion=Locked
 }
 {
  bone1Name=dyn_jacket_sd_02_l
  bone2Name=dyn_jacket_sd_01_l
  swing1LimitDegrees=15
  swing2LimitDegrees=15
  twistLimitDegrees=5
  twistMotion=Locked
 }
 {
  bone1Name=dyn_jacket_sd_03_l
  bone2Name=dyn_jacket_sd_02_l
  swing1LimitDegrees=15
  swing2LimitDegrees=15
  twistLimitDegrees=5
  twistMotion=Locked
 }
 {
  bone1Name=dyn_jacket_sd_04_l
  bone2Name=dyn_jacket_sd_03_l
  swing1LimitDegrees=15
  swing2LimitDegrees=15
  twistLimitDegrees=5
  twistMotion=Locked
 }
 {
  bone1Name=dyn_jacket_sd_01_r
  bone2Name=base_jacket_sd_r
  swing1LimitDegrees=10
  swing2LimitDegrees=15
  twistLimitDegrees=45
  twistMotion=Locked
 }
 {
  bone1Name=dyn_jacket_sd_02_r
  bone2Name=dyn_jacket_sd_01_r
  swing1LimitDegrees=5
  swing2LimitDegrees=15
  twistLimitDegrees=5
  twistMotion=Locked
 }
 {
  bone1Name=dyn_jacket_sd_03_r
  bone2Name=dyn_jacket_sd_02_r
  swing1LimitDegrees=5
  swing2LimitDegrees=15
  twistLimitDegrees=5
  twistMotion=Locked
 }
 {
  bone1Name=dyn_jacket_sd_04_r
  bone2Name=dyn_jacket_sd_03_r
  swing1LimitDegrees=5
  swing2LimitDegrees=15
  twistLimitDegrees=5
  twistMotion=Locked
 }
 {
  bone1Name=dyn_jacket_fr_01_l
  bone2Name=base_jacket_fr_l
  swing1LimitDegrees=20
  swing2LimitDegrees=15
  twistLimitDegrees=5
  twistMotion=Locked
 }
 {bone1Name=dyn_jacket_fr_01_r, bone2Name=base_jacket_fr_r, swing1LimitDegrees=20, swing2LimitDegrees=10, twistLimitDegrees=5}
]
