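# Physics asset for the male medium PeaceScarf cosmetic body, named as its emote variant, for simulating the jacket.
# Seven back, side and front chains of four capsule bodies each hang from tiny base anchor bodies.
# Constraints link each segment to its parent with twist locked and small swing angle values, widest at the front.
# Capsules on the pelvis, lower spine, thighs and calves have no constraints and appear to act as body colliders.
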PhysicsAsset /BRCosmetics/Characters/Player/Male/Medium/Bodies/M_MED_PeaceScarf/Meshes/M_MED_PeaceScarf_Physics_Emote
bodies {
 base_jacket_bk_c [{shapeType=Capsule, length=0.505, radius=0.505, shapeName=base_jacket_bk_c_capsule, rotation=<rot 0 -90 0>}]
 base_jacket_bk_l [{shapeType=Capsule, length=0.505, radius=0.505, shapeName=base_jacket_bk_l_capsule, rotation=<rot 0 -90 0>}]
 base_jacket_bk_r [{shapeType=Capsule, length=0.505, radius=0.505, shapeName=base_jacket_bk_r_capsule, rotation=<rot 0 -90 0>}]
 base_jacket_fr_l [{shapeType=Capsule, length=0.505, radius=0.505, shapeName=base_jacket_fr_l_capsule, rotation=<rot 0 -90 0>}]
 base_jacket_fr_r [{shapeType=Capsule, length=0.505, radius=0.505, shapeName=base_jacket_fr_r_capsule, rotation=<rot 0 -90 0>}]
 base_jacket_sd_l [{shapeType=Capsule, length=0.505, radius=0.505, shapeName=base_jacket_sd_l_capsule, rotation=<rot 0 -90 0>}]
 base_jacket_sd_r [{shapeType=Capsule, length=0.505, radius=0.505, shapeName=base_jacket_sd_r_capsule, rotation=<rot 0 -90 0>}]
 calf_l [{shapeType=Capsule, length=34.7774, radius=8.0569, shapeName=calf_l_capsule, center=<-29.2102 -1.2985 1.1512>, rotation=<rot 90 0 0>}]
 calf_r [{shapeType=Capsule, length=34.7774, radius=8.0569, shapeName=calf_r_capsule, center=<29.2102 1.2985 -1.1512>, rotation=<rot 90 180 0>}]
 dyn_jacket_bk_01_c [{shapeType=Capsule, length=12, radius=8, shapeName=dyn_jacket_bk_01_c_capsule, center=<8.6056 0 6.9873>, rotation=<rot 90 -90 0>}]
 dyn_jacket_bk_01_l [{shapeType=Capsule, length=12, radius=8, shapeName=dyn_jacket_bk_01_l_capsule, center=<8.6056 0 6.9873>, rotation=<rot 90 -90 0>}]
 dyn_jacket_bk_01_r [{shapeType=Capsule, length=12, radius=8, shapeName=dyn_jacket_bk_01_r_capsule, center=<8.6056 0 6.9873>, rotation=<rot 90 -90 0>}]
 dyn_jacket_bk_02_c [{shapeType=Capsule, length=12, radius=8, shapeName=dyn_jacket_bk_02_c_capsule, center=<8.6056 0 6.9873>, rotation=<rot 90 -90 0>}]
 dyn_jacket_bk_02_l [{shapeType=Capsule, length=12, radius=8, shapeName=dyn_jacket_bk_02_l_capsule, center=<8.6056 0 6.9873>, rotation=<rot 90 -90 0>}]
 dyn_jacket_bk_02_r [{shapeType=Capsule, length=12, radius=8, shapeName=dyn_jacket_bk_02_r_capsule, center=<8.6056 0 6.9873>, rotation=<rot 90 -90 0>}]
 dyn_jacket_bk_03_c [{shapeType=Capsule, length=12, radius=8, shapeName=dyn_jacket_bk_03_c_capsule, center=<8.6056 0 6.9873>, rotation=<rot 90 -90 0>}]
 dyn_jacket_bk_03_l [{shapeType=Capsule, length=12, radius=8, shapeName=dyn_jacket_bk_03_l_capsule, center=<8.6056 0 6.9873>, rotation=<rot 90 -90 0>}]
 dyn_jacket_bk_03_r [{shapeType=Capsule, length=12, radius=8, shapeName=dyn_jacket_bk_03_r_capsule, center=<8.6056 0 6.9873>, rotation=<rot 90 -90 0>}]
 dyn_jacket_bk_04_c [{shapeType=Capsule, length=12, radius=8, shapeName=dyn_jacket_bk_04_c_capsule, center=<8.6056 0 6.9873>, rotation=<rot 90 -90 0>}]
 dyn_jacket_bk_04_l [{shapeType=Capsule, length=12, radius=8, shapeName=dyn_jacket_bk_04_l_capsule, center=<8.6056 0 6.9873>, rotation=<rot 90 -90 0>}]
 dyn_jacket_bk_04_r [{shapeType=Capsule, length=12, radius=8, shapeName=dyn_jacket_bk_04_r_capsule, center=<8.6056 0 6.9873>, rotation=<rot 90 -90 0>}]
 dyn_jacket_fr_01_l [{
   shapeType=Capsule
   length=12
   radius=8
   shapeName=dyn_jacket_fr_01_l_capsule
   center=<8.6056 -7.6693 6.9873>
   rotation=<rot 88.4899 -90.0101 -0.0103>
  }]
 dyn_jacket_fr_01_r [{
   shapeType=Capsule
   length=12
   radius=8
   shapeName=dyn_jacket_fr_01_r_capsule
   center=<8.6056 6.1707 6.9873>
   rotation=<rot 86.2525 90.003 -179.9968>
  }]
 dyn_jacket_fr_02_l [{
   shapeType=Capsule
   length=12
   radius=8
   shapeName=dyn_jacket_fr_02_l_capsule
   center=<8.6056 -6.6197 6.5739>
   rotation=<rot 89.3648 -90.0341 -0.0344>
  }]
 dyn_jacket_fr_02_r [{
   shapeType=Capsule
   length=12
   radius=8
   shapeName=dyn_jacket_fr_02_r_capsule
   center=<8.6056 6.1707 6.9873>
   rotation=<rot 86.2525 90.003 -179.9968>
  }]
 dyn_jacket_fr_03_l [{
   shapeType=Capsule
   length=12
   radius=8
   shapeName=dyn_jacket_fr_03_l_capsule
   center=<8.6056 -7.6297 5.4854>
   rotation=<rot 88.4899 -90.0098 -0.01>
  }]
 dyn_jacket_fr_03_r [{
   shapeType=Capsule
   length=12
   radius=8
   shapeName=dyn_jacket_fr_03_r_capsule
   center=<8.6056 6.1145 6.1282>
   rotation=<rot 86.2525 90.003 -179.9968>
  }]
 dyn_jacket_fr_04_l [{
   shapeType=Capsule
   length=12
   radius=8
   shapeName=dyn_jacket_fr_04_l_capsule
   center=<8.6056 -7.6297 5.4838>
   rotation=<rot 88.4899 -90.0098 -0.01>
  }]
 dyn_jacket_fr_04_r [{
   shapeType=Capsule
   length=13.7293
   radius=8
   shapeName=dyn_jacket_fr_04_r_capsule
   center=<8.6056 5.961 5.2543>
   rotation=<rot 81.1549 90.0018 -179.998>
  }]
 dyn_jacket_sd_01_l [{shapeType=Capsule, length=12, radius=8, shapeName=dyn_jacket_sd_01_l_capsule, center=<8.6056 0 6.9873>, rotation=<rot 90 -90 0>}]
 dyn_jacket_sd_01_r [{shapeType=Capsule, length=12, radius=8, shapeName=dyn_jacket_sd_01_r_capsule, center=<8.6056 0 6.9873>, rotation=<rot 90 -90 0>}]
 dyn_jacket_sd_02_l [{shapeType=Capsule, length=12, radius=8, shapeName=dyn_jacket_sd_02_l_capsule, center=<8.6056 0 6.9873>, rotation=<rot 90 -90 0>}]
 dyn_jacket_sd_02_r [{shapeType=Capsule, length=15.4759, radius=8, shapeName=dyn_jacket_sd_02_r_capsule, center=<8.6056 0 6.1102>, rotation=<rot 90 -90 0>}]
 dyn_jacket_sd_03_l [{shapeType=Capsule, length=12, radius=8, shapeName=dyn_jacket_sd_03_l_capsule, center=<8.6056 0 5.1589>, rotation=<rot 90 -90 0>}]
 dyn_jacket_sd_03_r [{shapeType=Capsule, length=15.4759, radius=8, shapeName=dyn_jacket_sd_03_r_capsule, center=<8.6056 0 4.032>, rotation=<rot 90 -90 0>}]
 dyn_jacket_sd_04_l [{shapeType=Capsule, length=12, radius=8, shapeName=dyn_jacket_sd_04_l_capsule, center=<8.6056 0 5.1589>, rotation=<rot 90 -90 0>}]
 dyn_jacket_sd_04_r [{shapeType=Capsule, length=14.6134, radius=8, shapeName=dyn_jacket_sd_04_r_capsule, center=<8.6056 0 4.4292>, rotation=<rot 90 -90 0>}]
 pelvis [{shapeType=Capsule, length=10.2603, radius=8.0704, shapeName=pelvis_capsule, center=<0 -1.0694 0>}]
 spine_01 [{shapeType=Capsule, length=10.6469, radius=7.2325, shapeName=spine_01_capsule, center=<0 -0.6262 0>, rotation=<rot 0 -90 0>}]
 spine_02 [{shapeType=Capsule, length=7.8853, radius=7.7879, shapeName=spine_02_capsule, center=<0 -0.1467 0>}]
 thigh_l [{shapeType=Capsule, length=43.6009, radius=6.7908, shapeName=thigh_l_capsule, center=<-19.5126 0 0.7533>, rotation=<rot 90 0 0>}]
 thigh_r [{shapeType=Capsule, length=43.6009, radius=6.7908, shapeName=thigh_r_capsule, center=<19.5126 0 -0.7533>, rotation=<rot 90 180 0>}]
}
constraints [
 {
  bone1Name=dyn_jacket_bk_01_c
  bone2Name=base_jacket_bk_c
  swing1LimitDegrees=5
  swing2LimitDegrees=15
  twistLimitDegrees=45
  twistMotion=Locked
 }
 {
  bone1Name=dyn_jacket_bk_02_c
  bone2Name=dyn_jacket_bk_01_c
  swing1LimitDegrees=5
  swing2LimitDegrees=15
  twistLimitDegrees=5
  twistMotion=Locked
 }
 {
  bone1Name=dyn_jacket_bk_03_c
  bone2Name=dyn_jacket_bk_02_c
  swing1LimitDegrees=5
  swing2LimitDegrees=15
  twistLimitDegrees=5
  twistMotion=Locked
 }
 {
  bone1Name=dyn_jacket_bk_04_c
  bone2Name=dyn_jacket_bk_03_c
  swing1LimitDegrees=5
  swing2LimitDegrees=15
  twistLimitDegrees=5
  twistMotion=Locked
 }
 {
  bone1Name=dyn_jacket_bk_01_l
  bone2Name=base_jacket_bk_l
  swing1LimitDegrees=5
  swing2LimitDegrees=15
  twistLimitDegrees=45
  twistMotion=Locked
 }
 {
  bone1Name=dyn_jacket_bk_02_l
  bone2Name=dyn_jacket_bk_01_l
  swing1LimitDegrees=5
  swing2LimitDegrees=15
  twistLimitDegrees=5
  twistMotion=Locked
 }
 {
  bone1Name=dyn_jacket_bk_03_l
  bone2Name=dyn_jacket_bk_02_l
  swing1LimitDegrees=5
  swing2LimitDegrees=15
  twistLimitDegrees=5
  twistMotion=Locked
 }
 {
  bone1Name=dyn_jacket_bk_04_l
  bone2Name=dyn_jacket_bk_03_l
  swing1LimitDegrees=5
  swing2LimitDegrees=15
  twistLimitDegrees=5
  twistMotion=Locked
 }
 {
  bone1Name=dyn_jacket_bk_01_r
  bone2Name=base_jacket_bk_r
  swing1LimitDegrees=5
  swing2LimitDegrees=15
  twistLimitDegrees=45
  twistMotion=Locked
 }
 {
  bone1Name=dyn_jacket_bk_02_r
  bone2Name=dyn_jacket_bk_01_r
  swing1LimitDegrees=5
  swing2LimitDegrees=15
  twistLimitDegrees=5
  twistMotion=Locked
 }
 {
  bone1Name=dyn_jacket_bk_03_r
  bone2Name=dyn_jacket_bk_02_r
  swing1LimitDegrees=5
  swing2LimitDegrees=15
  twistLimitDegrees=5
  twistMotion=Locked
 }
 {
  bone1Name=dyn_jacket_bk_04_r
  bone2Name=dyn_jacket_bk_03_r
  swing1LimitDegrees=5
  swing2LimitDegrees=15
  twistLimitDegrees=5
  twistMotion=Locked
 }
 {
  bone1Name=dyn_jacket_fr_01_l
  bone2Name=base_jacket_fr_l
  swing1LimitDegrees=15
  swing2LimitDegrees=20
  twistLimitDegrees=45
  twistMotion=Locked
 }
 {
  bone1Name=dyn_jacket_fr_02_l
  bone2Name=dyn_jacket_fr_01_l
  swing1LimitDegrees=15
  swing2LimitDegrees=20
  twistLimitDegrees=5
  twistMotion=Locked
 }
 {
  bone1Name=dyn_jacket_fr_03_l
  bone2Name=dyn_jacket_fr_02_l
  swing1LimitDegrees=15
  swing2LimitDegrees=20
  twistLimitDegrees=5
  twistMotion=Locked
 }
 {
  bone1Name=dyn_jacket_fr_04_l
  bone2Name=dyn_jacket_fr_03_l
  swing1LimitDegrees=15
  swing2LimitDegrees=20
  twistLimitDegrees=5
  twistMotion=Locked
 }
 {
  bone1Name=dyn_jacket_fr_01_r
  bone2Name=base_jacket_fr_r
  swing1LimitDegrees=15
  swing2LimitDegrees=20
  twistLimitDegrees=45
  twistMotion=Locked
 }
 {
  bone1Name=dyn_jacket_fr_02_r
  bone2Name=dyn_jacket_fr_01_r
  swing1LimitDegrees=15
  swing2LimitDegrees=20
  twistLimitDegrees=5
  twistMotion=Locked
 }
 {
  bone1Name=dyn_jacket_fr_03_r
  bone2Name=dyn_jacket_fr_02_r
  swing1LimitDegrees=15
  swing2LimitDegrees=20
  twistLimitDegrees=5
  twistMotion=Locked
 }
 {
  bone1Name=dyn_jacket_fr_04_r
  bone2Name=dyn_jacket_fr_03_r
  swing1LimitDegrees=15
  swing2LimitDegrees=20
  twistLimitDegrees=5
  twistMotion=Locked
 }
 {
  bone1Name=dyn_jacket_sd_01_l
  bone2Name=base_jacket_sd_l
  swing1LimitDegrees=12
  swing2LimitDegrees=15
  twistLimitDegrees=45
  twistMotion=Locked
 }
 {
  bone1Name=dyn_jacket_sd_02_l
  bone2Name=dyn_jacket_sd_01_l
  swing1LimitDegrees=12
  swing2LimitDegrees=15
  twistLimitDegrees=5
  twistMotion=Locked
 }
 {
  bone1Name=dyn_jacket_sd_03_l
  bone2Name=dyn_jacket_sd_02_l
  swing1LimitDegrees=12
  swing2LimitDegrees=15
  twistLimitDegrees=5
  twistMotion=Locked
 }
 {
  bone1Name=dyn_jacket_sd_04_l
  bone2Name=dyn_jacket_sd_03_l
  swing1LimitDegrees=12
  swing2LimitDegrees=15
  twistLimitDegrees=5
  twistMotion=Locked
 }
 {
  bone1Name=dyn_jacket_sd_01_r
  bone2Name=base_jacket_sd_r
  swing1LimitDegrees=10
  swing2LimitDegrees=15
  twistLimitDegrees=45
  twistMotion=Locked
 }
 {
  bone1Name=dyn_jacket_sd_02_r
  bone2Name=dyn_jacket_sd_01_r
  swing1LimitDegrees=10
  swing2LimitDegrees=15
  twistLimitDegrees=5
  twistMotion=Locked
 }
 {
  bone1Name=dyn_jacket_sd_03_r
  bone2Name=dyn_jacket_sd_02_r
  swing1LimitDegrees=10
  swing2LimitDegrees=15
  twistLimitDegrees=5
  twistMotion=Locked
 }
 {
  bone1Name=dyn_jacket_sd_04_r
  bone2Name=dyn_jacket_sd_03_r
  swing1LimitDegrees=10
  swing2LimitDegrees=15
  twistLimitDegrees=5
  twistMotion=Locked
 }
]
