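# Animation Blueprint for the male medium PeaceScarf cosmetic body, adding secondary motion to its jacket bones.
# It copies the parent mesh pose through a state-driven Control Rig, then blends between emote, crouch and default RigidBody branches, the default adding look-at and additive bone corrections.
# The event graph derives state flags, gravity and physics settings, a sine-wave jacket rotation for sliding and grinding, and alternative skin-weight profiles when rigid body is disabled.
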AnimBlueprint /BRCosmetics/Characters/Player/Male/Medium/Bodies/M_MED_PeaceScarf/Meshes/M_MED_PeaceScarf_AnimBP
target_skeleton=&/M_MED_PeaceScarf_Skeleton
graphs [
 AnimGraph {nodes [
   n1 = "Transform (Modify) Bone - Bone: base_jacket_fr_r"(
    ComponentPose=Local To Component(LocalPose=Use cached pose 'CtrlRigCache'())
    Alpha=1
    Translation="0,0,0"
    Rotation="4.320,25.077,2.019"
    Scale="1,1,1"
   ) {
    rotationMode=BMM_Additive
    rotationSpace=BCS_BoneSpace
    translationMode=BMM_Additive
    translationSpace=BCS_BoneSpace
    boneToModify {boneName=base_jacket_fr_r}
    rotation=<rot 4.32 25.077 2.019>
   }
   n2 = RigidBody!(ComponentPose=@n1, Alpha, OverrideWorldGravity:Vector="0,0,-980", ExternalForce) -> (Pose) {
    bOverrideWorldGravity=true
    overridePhysicsAsset=&/M_MED_PeaceScarf_Physics_Crouch
    simulationSpace=BaseBoneSpace
    baseBoneRef {boneName=pelvis}
    componentAppliedLinearAccClamp=<3000 3000 3000>
    simSpaceSettings {
     maxAngularAcceleration=50
     maxAngularVelocity=50
     maxLinearAcceleration=200
     maxLinearVelocity=100
     worldAlpha=0.2
     externalLinearDragV=<3 3 3>
    }
   }
   n3 = Component To Local(ComponentPose=@n2)
   n4 = Look At - Bone: base_jacket_fr_r!(ComponentPose=Local To Component(LocalPose=Use cached pose 'CtrlRigCache'()), bAlphaBoolEnabled:bool=True) -> (Pose) {
    alphaInputType=Bool
    alphaBoolBlend {blendInTime=0.3, blendOutTime=0.2}
    boneToModify {boneName=base_jacket_fr_r}
    lookAtLocation=<12.2412 31.2763 41.4493>
    lookAtTarget {boneReference {boneName=foot_r}}
    lookAt_Axis {axis {x=1}}
   }
   n5 = Look At - Bone: base_jacket_sd_r!(ComponentPose=@n4, bAlphaBoolEnabled:bool=True) -> (Pose) {
    alphaInputType=Bool
    alphaBoolBlend {blendInTime=0.3, blendOutTime=0.2}
    boneToModify {boneName=base_jacket_sd_r}
    lookAtLocation=<6.2777 49.0524 -25.8844>
    lookAtTarget {boneReference {boneName=foot_r}}
    lookAt_Axis {axis {x=1}}
   }
   n6 = "Transform (Modify) Bone - Bone: dyn_jacket_bk_02_r"(ComponentPose=@n5, Alpha=1, Translation="0,0,0", Rotation="-30,0,0", Scale="1,1,1") {
    rotationMode=BMM_Additive
    rotationSpace=BCS_BoneSpace
    translationMode=BMM_Additive
    translationSpace=BCS_BoneSpace
    alphaBoolBlend {blendInTime=0.5, blendOutTime=0.2}
    alphaScaleBiasClamp {bInterpResult=true}
    boneToModify {boneName=dyn_jacket_bk_02_r}
    rotation {pitch=-30}
   }
   n7 = "Transform (Modify) Bone - Bone: dyn_jacket_bk_02_c"(ComponentPose=@n6, Alpha=1, Translation="0,0,0", Rotation="-30,0,0", Scale="1,1,1") {
    rotationMode=BMM_Additive
    rotationSpace=BCS_BoneSpace
    translationMode=BMM_Additive
    translationSpace=BCS_BoneSpace
    alphaBoolBlend {blendInTime=0.5, blendOutTime=0.2}
    alphaScaleBiasClamp {bInterpResult=true}
    boneToModify {boneName=dyn_jacket_bk_02_c}
    rotation {pitch=-30}
   }
   n8 = "Transform (Modify) Bone - Bone: dyn_jacket_bk_02_l"(ComponentPose=@n7, Alpha=1, Translation="0,0,0", Rotation="-30,0,0", Scale="1,1,1") {
    rotationMode=BMM_Additive
    rotationSpace=BCS_BoneSpace
    translationMode=BMM_Additive
    translationSpace=BCS_BoneSpace
    alphaBoolBlend {blendInTime=0.5, blendOutTime=0.2}
    alphaScaleBiasClamp {bInterpResult=true}
    boneToModify {boneName=dyn_jacket_bk_02_l}
    rotation {pitch=-30}
   }
   n9 = RigidBody!(
    ComponentPose=@n8
    Alpha
    OverrideWorldGravity="0,0,0"
    ExternalForce
    SimSpaceSettings:Sim Space Settings Structure=(WorldAlpha=0,VelocityScaleZ=1,DampingAlpha=1,MaxLinearVelocity=200,MaxAngularVelocity=50,MaxLinearAcceleration=400,MaxAngularAcceleration=150,ExternalLinearDragV=(X=1,Y=1,Z=1),ExternalLinearVelocity=(X=0,Y=0,Z=0),ExternalAngularVelocity=(X=0,Y=0,Z=0))
   ) -> (Pose) {
    bOverrideWorldGravity=true
    overridePhysicsAsset=&/M_MED_PeaceScarf_Physics
    simulationSpace=BaseBoneSpace
    alphaScaleBiasClamp {bInterpResult=true, interpSpeedDecreasing=5}
    baseBoneRef {boneName=spine_03}
    componentAppliedLinearAccClamp=<3000 3000 3000>
    simSpaceSettings {maxAngularAcceleration=150, maxAngularVelocity=50, maxLinearAcceleration=400, maxLinearVelocity=200, externalLinearDragV=<1 1 1>}
   }
   n10 = "Transform (Modify) Bone - Bone: base_jacket_fr_l"!(ComponentPose=@n9, bAlphaBoolEnabled:bool=True, Translation="0,0,0", Rotation="10,0,0", Scale="1,1,1") -> (Pose) {
    alphaInputType=Bool
    rotationMode=BMM_Additive
    rotationSpace=BCS_BoneSpace
    translationMode=BMM_Additive
    translationSpace=BCS_BoneSpace
    alphaBoolBlend {blendInTime=0.1, blendOption=Linear, blendOutTime=0.2}
    boneToModify {boneName=base_jacket_fr_l}
    rotation {pitch=10}
   }
   n11 = "Transform (Modify) Bone - Bone: base_jacket_bk_c"!(ComponentPose=@n10, bAlphaBoolEnabled:bool=True, Translation="0,0,0", Rotation="15,0,0", Scale="1,1,1") -> (Pose) {
    alphaInputType=Bool
    rotationMode=BMM_Additive
    rotationSpace=BCS_BoneSpace
    translationMode=BMM_Additive
    translationSpace=BCS_BoneSpace
    alphaBoolBlend {blendInTime=0.2, blendOutTime=0.2}
    boneToModify {boneName=base_jacket_bk_c}
    rotation {pitch=15}
   }
   n12 = "Transform (Modify) Bone - Bone: base_jacket_bk_r"!(ComponentPose=@n11, bAlphaBoolEnabled:bool=True, Translation="0,0,0", Rotation="15,0,0", Scale="1,1,1") -> (Pose) {
    alphaInputType=Bool
    rotationMode=BMM_Additive
    rotationSpace=BCS_BoneSpace
    translationMode=BMM_Additive
    translationSpace=BCS_BoneSpace
    alphaBoolBlend {blendInTime=0.2, blendOutTime=0.2}
    boneToModify {boneName=base_jacket_bk_r}
    rotation {pitch=15}
   }
   n13 = "Transform (Modify) Bone - Bone: base_jacket_sd_l"!(ComponentPose=@n12, bAlphaBoolEnabled:bool=True, Translation="0,0,0", Rotation="-7,0,0", Scale="1,1,1") -> (Pose) {
    alphaInputType=Bool
    rotationMode=BMM_Additive
    rotationSpace=BCS_BoneSpace
    translationMode=BMM_Additive
    translationSpace=BCS_BoneSpace
    alphaBoolBlend {blendInTime=0.2, blendOutTime=0.2}
    boneToModify {boneName=base_jacket_sd_l}
    rotation {pitch=-7}
   }
   n14 = "Transform (Modify) Bone - Bone: base_jacket_sd_r"!(ComponentPose=@n13, bAlphaBoolEnabled:bool=True, Translation="0,0,0", Rotation="-10,0,0", Scale="1,1,1") -> (Pose) {
    alphaInputType=Bool
    rotationMode=BMM_Additive
    rotationSpace=BCS_BoneSpace
    translationMode=BMM_Additive
    translationSpace=BCS_BoneSpace
    alphaBoolBlend {blendInTime=0.5, blendOutTime=0.2}
    boneToModify {boneName=base_jacket_sd_r}
    rotation {pitch=-10}
   }
   n15 = "Transform (Modify) Bone - Bone: dyn_jacket_fr_02_r"!(ComponentPose=@n14, bAlphaBoolEnabled:bool=True, Translation="0,0,0", Rotation="-16,-4.6,0", Scale="1,1,1") -> (Pose) {
    alphaInputType=Bool
    rotationMode=BMM_Additive
    rotationSpace=BCS_BoneSpace
    translationMode=BMM_Additive
    translationSpace=BCS_BoneSpace
    alphaBoolBlend {blendInTime=0.5, blendOutTime=0.2}
    boneToModify {boneName=dyn_jacket_fr_02_r}
    rotation {pitch=-16, yaw=-4.6}
   }
   n16 = "Transform (Modify) Bone - Bone: base_jacket_bk_c"(ComponentPose=@n15, bAlphaBoolEnabled=True, Translation="0,0,0", Rotation="20,0,0", Scale="1,1,1") {
    alphaInputType=Bool
    rotationMode=BMM_Additive
    rotationSpace=BCS_BoneSpace
    translationMode=BMM_Additive
    translationSpace=BCS_BoneSpace
    alphaBoolBlend {blendInTime=0.5, blendOutTime=0.2}
    boneToModify {boneName=base_jacket_bk_c}
    rotation {pitch=20}
   }
   n17 = "Transform (Modify) Bone - Bone: base_jacket_bk_r"(ComponentPose=@n16, bAlphaBoolEnabled=True, Translation="0,0,0", Rotation="15,-0.229,-10", Scale="1,1,1") {
    alphaInputType=Bool
    rotationMode=BMM_Additive
    rotationSpace=BCS_BoneSpace
    translationMode=BMM_Additive
    translationSpace=BCS_BoneSpace
    alphaBoolBlend {blendInTime=0.5, blendOutTime=0.2}
    boneToModify {boneName=base_jacket_bk_r}
    rotation=<rot 15 -0.229 -10>
   }
   n18 = "Transform (Modify) Bone - Bone: base_jacket_bk_l"!(ComponentPose=@n17, bAlphaBoolEnabled:bool=True, Translation="0,0,0", Rotation="15,0,0", Scale="1,1,1") -> (Pose) {
    alphaInputType=Bool
    rotationMode=BMM_Additive
    rotationSpace=BCS_BoneSpace
    translationMode=BMM_Additive
    translationSpace=BCS_BoneSpace
    alphaBoolBlend {blendInTime=0.5, blendOutTime=0.2}
    boneToModify {boneName=base_jacket_bk_l}
    rotation {pitch=15}
   }
   n19 = "Transform (Modify) Bone - Bone: base_jacket_sd_r"(ComponentPose=@n18, bAlphaBoolEnabled=True, Translation="0,0,0", Rotation="15,0,0", Scale="1,1,1") {
    alphaInputType=Bool
    rotationMode=BMM_Additive
    rotationSpace=BCS_BoneSpace
    translationMode=BMM_Additive
    translationSpace=BCS_BoneSpace
    alphaBoolBlend {blendInTime=0.5, blendOutTime=0.2}
    boneToModify {boneName=base_jacket_sd_r}
    rotation {pitch=15}
   }
   n20 = "Transform (Modify) Bone - Bone: dyn_jacket_bk_02_c"!(ComponentPose=@n19, bAlphaBoolEnabled:bool=True, Translation="0,0,0", Rotation="20,0,0", Scale="1,1,1") -> (Pose) {
    alphaInputType=Bool
    rotationMode=BMM_Additive
    rotationSpace=BCS_BoneSpace
    translationMode=BMM_Additive
    translationSpace=BCS_BoneSpace
    alphaBoolBlend {blendInTime=0.5, blendOutTime=0.2}
    boneToModify {boneName=dyn_jacket_bk_02_c}
    rotation {pitch=20}
   }
   n21 = "Transform (Modify) Bone - Bone: dyn_jacket_bk_02_r"!(ComponentPose=@n20, bAlphaBoolEnabled:bool=True, Translation="0,0,0", Rotation="15,0,0", Scale="1,1,1") -> (Pose) {
    alphaInputType=Bool
    rotationMode=BMM_Additive
    rotationSpace=BCS_BoneSpace
    translationMode=BMM_Additive
    translationSpace=BCS_BoneSpace
    alphaBoolBlend {blendInTime=0.5, blendOutTime=0.2}
    boneToModify {boneName=dyn_jacket_bk_02_r}
    rotation {pitch=15}
   }
   n22 = "Transform (Modify) Bone - Bone: dyn_jacket_bk_02_l"!(ComponentPose=@n21, bAlphaBoolEnabled:bool=True, Translation="0,0,0", Rotation="15,0,0", Scale="1,1,1") -> (Pose) {
    alphaInputType=Bool
    rotationMode=BMM_Additive
    rotationSpace=BCS_BoneSpace
    translationMode=BMM_Additive
    translationSpace=BCS_BoneSpace
    alphaBoolBlend {blendInTime=0.5, blendOutTime=0.2}
    boneToModify {boneName=dyn_jacket_bk_02_l}
    rotation {pitch=15}
   }
   n23 = "Transform (Modify) Bone - Bone: dyn_jacket_sd_02_r"!(ComponentPose=@n22, bAlphaBoolEnabled:bool=True, Translation="0,0,0", Rotation="15,0,0", Scale="1,1,1") -> (Pose) {
    alphaInputType=Bool
    rotationMode=BMM_Additive
    rotationSpace=BCS_BoneSpace
    translationMode=BMM_Additive
    translationSpace=BCS_BoneSpace
    alphaBoolBlend {blendInTime=0.5, blendOutTime=0.2}
    boneToModify {boneName=dyn_jacket_sd_02_r}
    rotation {pitch=15}
   }
   n24 = "Transform (Modify) Bone - Bone: dyn_jacket_bk_03_c"!(ComponentPose=@n23, bAlphaBoolEnabled:bool=True, Translation="0,0,0", Rotation="15,0,0", Scale="1,1,1") -> (Pose) {
    alphaInputType=Bool
    rotationMode=BMM_Additive
    rotationSpace=BCS_BoneSpace
    translationMode=BMM_Additive
    translationSpace=BCS_BoneSpace
    alphaBoolBlend {blendInTime=0.5, blendOutTime=0.2}
    boneToModify {boneName=dyn_jacket_bk_03_c}
    rotation {pitch=15}
   }
   n25 = "Transform (Modify) Bone - Bone: dyn_jacket_bk_03_r"!(ComponentPose=@n24, bAlphaBoolEnabled:bool=True, Translation="0,0,0", Rotation="15,0,0", Scale="1,1,1") -> (Pose) {
    alphaInputType=Bool
    rotationMode=BMM_Additive
    rotationSpace=BCS_BoneSpace
    translationMode=BMM_Additive
    translationSpace=BCS_BoneSpace
    alphaBoolBlend {blendInTime=0.5, blendOutTime=0.2}
    boneToModify {boneName=dyn_jacket_bk_03_r}
    rotation {pitch=15}
   }
   n26 = "Transform (Modify) Bone - Bone: dyn_jacket_bk_03_l"!(ComponentPose=@n25, bAlphaBoolEnabled:bool=True, Translation="0,0,0", Rotation="15,0,0", Scale="1,1,1") -> (Pose) {
    alphaInputType=Bool
    rotationMode=BMM_Additive
    rotationSpace=BCS_BoneSpace
    translationMode=BMM_Additive
    translationSpace=BCS_BoneSpace
    alphaBoolBlend {blendInTime=0.5, blendOutTime=0.2}
    boneToModify {boneName=dyn_jacket_bk_03_l}
    rotation {pitch=15}
   }
   n27 = "Transform (Modify) Bone - Bone: dyn_jacket_bk_03_r"(ComponentPose=@n26, bAlphaBoolEnabled=True, Translation="0,0,0", Rotation="15,0,0", Scale="1,1,1") {
    alphaInputType=Bool
    rotationMode=BMM_Additive
    rotationSpace=BCS_BoneSpace
    translationMode=BMM_Additive
    translationSpace=BCS_BoneSpace
    alphaBoolBlend {blendInTime=0.5, blendOutTime=0.2}
    boneToModify {boneName=dyn_jacket_bk_03_r}
    rotation {pitch=15}
   }
   n28 = "Transform (Modify) Bone - Bone: base_jacket_bk_c"(ComponentPose=@n27, bAlphaBoolEnabled=True, Translation="3,0,0", Rotation="0,0,0", Scale="1,1,1") {
    alphaInputType=Bool
    rotationMode=BMM_Additive
    rotationSpace=BCS_BoneSpace
    translationMode=BMM_Additive
    translationSpace=BCS_BoneSpace
    alphaBoolBlend {blendInTime=0.2, blendOutTime=0.2}
    alphaScaleBiasClamp {bInterpResult=true, interpSpeedIncreasing=1}
    boneToModify {boneName=base_jacket_bk_c}
    translation {x=3}
   }
   n29 = "Transform (Modify) Bone - Bone: base_jacket_bk_r"(ComponentPose=@n28, bAlphaBoolEnabled=True, Translation="3,0,0", Rotation="0,0,0", Scale="1,1,1") {
    alphaInputType=Bool
    rotationMode=BMM_Additive
    rotationSpace=BCS_BoneSpace
    translationMode=BMM_Additive
    translationSpace=BCS_BoneSpace
    alphaBoolBlend {blendInTime=0.2, blendOutTime=0.2}
    alphaScaleBiasClamp {bInterpResult=true, interpSpeedIncreasing=1}
    boneToModify {boneName=base_jacket_bk_r}
    translation {x=3}
   }
   n30 = "Transform (Modify) Bone - Bone: base_jacket_bk_l"(ComponentPose=@n29, bAlphaBoolEnabled=True, Translation="3,0,0", Rotation="0,0,0", Scale="1,1,1") {
    alphaInputType=Bool
    rotationMode=BMM_Additive
    rotationSpace=BCS_BoneSpace
    translationMode=BMM_Additive
    translationSpace=BCS_BoneSpace
    alphaBoolBlend {blendInTime=0.2, blendOutTime=0.2}
    alphaScaleBiasClamp {bInterpResult=true, interpSpeedIncreasing=1}
    boneToModify {boneName=base_jacket_bk_l}
    translation {x=3}
   }
   n31 = "Transform (Modify) Bone - Bone: base_jacket_fr_l"(ComponentPose=@n30, bAlphaBoolEnabled=True, Translation="0,0,0", Rotation="0,0,0", Scale="1,1,1") {
    alphaInputType=Bool
    rotationMode=BMM_Additive
    rotationSpace=BCS_BoneSpace
    translationMode=BMM_Additive
    translationSpace=BCS_BoneSpace
    alphaBoolBlend {blendInTime=0.4, blendOutTime=0.2}
    alphaScaleBiasClamp {bInterpResult=true, interpSpeedIncreasing=1.5}
    boneToModify {boneName=base_jacket_fr_l}
   }
   n32 = "Transform (Modify) Bone - Bone: dyn_jacket_sd_02_l"!(ComponentPose=@n31, bAlphaBoolEnabled:bool=True, Translation="0,0,0", Rotation="0,0,0", Scale="1,1,1") -> (Pose) {
    alphaInputType=Bool
    rotationMode=BMM_Additive
    rotationSpace=BCS_BoneSpace
    translationMode=BMM_Additive
    translationSpace=BCS_BoneSpace
    alphaBoolBlend {blendInTime=0.4, blendOutTime=0.2}
    alphaScaleBiasClamp {bInterpResult=true, interpSpeedIncreasing=1.5}
    boneToModify {boneName=dyn_jacket_sd_02_l}
   }
   n33 = "Transform (Modify) Bone - Bone: dyn_jacket_bk_02_l"(ComponentPose=@n32, bAlphaBoolEnabled=True, Translation="0,0,0", Rotation="0,0,0", Scale="1,1,1") {
    alphaInputType=Bool
    rotationMode=BMM_Additive
    rotationSpace=BCS_BoneSpace
    translationMode=BMM_Additive
    translationSpace=BCS_BoneSpace
    alphaBoolBlend {blendInTime=0.4, blendOutTime=0.2}
    alphaScaleBiasClamp {bInterpResult=true, interpSpeedIncreasing=1.5}
    boneToModify {boneName=dyn_jacket_bk_02_l}
   }
   n34 = "Transform (Modify) Bone - Bone: dyn_jacket_bk_02_c"(ComponentPose=@n33, bAlphaBoolEnabled=True, Translation="0,0,0", Rotation="0,0,0", Scale="1,1,1") {
    alphaInputType=Bool
    rotationMode=BMM_Additive
    rotationSpace=BCS_BoneSpace
    translationMode=BMM_Additive
    translationSpace=BCS_BoneSpace
    alphaBoolBlend {blendInTime=0.4, blendOutTime=0.2}
    alphaScaleBiasClamp {bInterpResult=true, interpSpeedIncreasing=1.5}
    boneToModify {boneName=dyn_jacket_bk_02_c}
   }
   n35 = "Transform (Modify) Bone - Bone: dyn_jacket_bk_02_r"(ComponentPose=@n34, bAlphaBoolEnabled=True, Translation="0,0,0", Rotation="0,0,0", Scale="1,1,1") {
    alphaInputType=Bool
    rotationMode=BMM_Additive
    rotationSpace=BCS_BoneSpace
    translationMode=BMM_Additive
    translationSpace=BCS_BoneSpace
    alphaBoolBlend {blendInTime=0.4, blendOutTime=0.2}
    alphaScaleBiasClamp {bInterpResult=true, interpSpeedIncreasing=1.5}
    boneToModify {boneName=dyn_jacket_bk_02_r}
   }
   n36 = "Transform (Modify) Bone - Bone: base_jacket_sd_r"(ComponentPose=@n35, bAlphaBoolEnabled=True, Translation="0,0,0", Rotation="0,0,0", Scale="1,1,1") {
    alphaInputType=Bool
    rotationMode=BMM_Additive
    rotationSpace=BCS_BoneSpace
    translationMode=BMM_Additive
    translationSpace=BCS_BoneSpace
    alphaBoolBlend {blendInTime=0.4, blendOutTime=0.2}
    alphaScaleBiasClamp {bInterpResult=true, interpSpeedIncreasing=1.5}
    boneToModify {boneName=base_jacket_sd_r}
   }
   n37 = "Transform (Modify) Bone - Bone: base_jacket_fr_r"!(ComponentPose=@n36, bAlphaBoolEnabled:bool=True, Translation="0,0,0", Rotation="0,0,0", Scale="1,1,1") -> (Pose) {
    alphaInputType=Bool
    rotationMode=BMM_Additive
    rotationSpace=BCS_BoneSpace
    translationMode=BMM_Additive
    translationSpace=BCS_BoneSpace
    alphaBoolBlend {blendInTime=0.4, blendOutTime=0.2}
    alphaScaleBiasClamp {bInterpResult=true, interpSpeedIncreasing=1.5}
    boneToModify {boneName=base_jacket_fr_r}
   }
   n38 = "Transform (Modify) Bone - Bone: dyn_jacket_bk_02_l"(ComponentPose=@n37, bAlphaBoolEnabled=True, Translation="0,0,0", Rotation="2.060,-1.546,-5.969", Scale="1,1,1") {
    alphaInputType=Bool
    rotationMode=BMM_Additive
    rotationSpace=BCS_BoneSpace
    translationMode=BMM_Additive
    translationSpace=BCS_BoneSpace
    alphaBoolBlend {blendInTime=0.4, blendOutTime=0.2}
    alphaScaleBiasClamp {bInterpResult=true, interpSpeedIncreasing=1.5}
    boneToModify {boneName=dyn_jacket_bk_02_l}
    rotation=<rot 2.06 -1.546 -5.969>
   }
   n39 = Component To Local(ComponentPose=@n38)
   n40 = Get OnlyCrouching() -> (OnlyCrouching:bool)
   n41 = RigidBody(
    ComponentPose=Local To Component(LocalPose=Use cached pose 'CtrlRigCache'())
    OverrideWorldGravity="0,0,0"
    SimSpaceSettings=(WorldAlpha=0,VelocityScaleZ=1,DampingAlpha=1,MaxLinearVelocity=200,MaxAngularVelocity=50,MaxLinearAcceleration=400,MaxAngularAcceleration=150,ExternalLinearDragV=(X=1,Y=1,Z=1),ExternalLinearVelocity=(X=0,Y=0,Z=0),ExternalAngularVelocity=(X=0,Y=0,Z=0))
   ) {
    bOverrideWorldGravity=true
    overridePhysicsAsset=&/M_MED_PeaceScarf_Physics_Emote
    simulationSpace=BaseBoneSpace
    baseBoneRef {boneName=spine_01}
    componentAppliedLinearAccClamp=<3000 3000 3000>
    simSpaceSettings {maxAngularAcceleration=150, maxAngularVelocity=50, maxLinearAcceleration=400, maxLinearVelocity=200, externalLinearDragV=<1 1 1>}
   }
   Output Pose(Result=Blend Poses by bool(
     BlendPose_0=Component To Local(ComponentPose=@n41)
     BlendPose_1=Blend Poses by bool(BlendPose_0=@n3, BlendPose_1=@n39, BlendTime_0=0.5, bActiveValue=@n40)
     bActiveValue=Get bIsPlayingEmote()
    )) {layerGroup=None}
   n42 = Control Rig!(
    Source=Copy Pose From Mesh() {bUseAttachedParent=true}
    Alpha
    CrouchingBool:bool=Get bIsCrouching()
    KneeSliding:bool=Get bIsKneeSliding()
    BaseballSliding:bool=Get bIsBaseballSliding()
    Pistol:double=Get Detected_Pistol() -> (Detected_Pistol:double)
    Emote:bool=Get bIsPlayingEmote()
    RBAN:bool=Get bRigidBodyEnabled()
    CrouchingIdle:double=Get CrouchingIdle() -> (CrouchingIdle:double)
    OnGrindRail:bool=Get bIsOnGrindRail()
    Targeting:bool=Get bIsTargeting()
    Swimming:bool=Get bIsSurfaceSwimming()
    BalletSpin:double=Get PlayingBalletSpin() -> (PlayingBalletSpin:double)
    Skydiving:bool=Get bIsSkyDiving()
    HeadBangerEmote:bool=Get PlayingHeadBanger() -> (PlayingHeadBanger:bool)
   ) -> (Pose) {controlRigAssetReference {blueprintRigClass=&/M_MED_PeaceScarf_CtrlRig.M_MED_PeaceScarf_CtrlRig_C}}
   Save cached pose 'CtrlRigCache'(Pose=@n42) {cacheName=CtrlRigCache}
  ]}
 EventGraph {nodes [
   Event BlueprintUpdateAnimation()
   > n1 = Sequence!(execute) -> (then_0, then_1, then_2:Exec, then_3:Exec, then_4:Exec, then_5:Exec, then_6:Exec, then_7:Exec)
   > n2 = "Gravity Override State Function (Parameter Asset) (Nativized)"(
    CharacterPartAnimInstance=Self-Reference()
    GameTimeInSeconds=GetGameTimeInSeconds()
    DebugJointName=spine_05
    Parameters=&/M_MED_PeaceScarf_Parameters
   )
   > Set Gravity Override(Gravity Override=@n2)
   > Set RBAN Settings Out(RBAN Settings Out=@n2.OutRigidBodyAnimNodeParameters)
   > Set Is Wearing Cape(Is Wearing Cape=CheckForBackbling().bIsBackblingACape)
   n3 = Get Is Wearing Cape() -> (Is Wearing Cape:bool)
   n4 = Get bIsJumping()
   > Set JumpingWithCape(JumpingWithCape:double="To Float (Boolean)"(InBool=AND Boolean(A=@n4, B=@n3))) -> (Output_Get:double)
   n5 = Branch(execute=@n1.then_5, Condition=OR Boolean(A=Get bIsSkyDiving(), B=Get PlayinhEasternBloc() -> (PlayinhEasternBloc:bool)))
   > Set Alpha_SkydiveFE And Emote(Alpha_SkydiveFE And Emote:double=0.5) -> (Output_Get:double)
   n6 = Branch(execute=@n5.else, Condition=Get PlayinhBreakDance() -> (PlayinhBreakDance:bool))
   > Set Alpha_SkydiveFE And Emote(Alpha_SkydiveFE And Emote=0.8)
   Set Alpha_SkydiveFE And Emote(execute=@n6.else, Alpha_SkydiveFE And Emote=0.9)
   n7 = NOT Boolean(A=Get bIsMoving2D())
   n8 = Get bIsCrouching()
   Set CrouchingIdle(execute=@n1.then_2, CrouchingIdle:double="To Float (Boolean)"(InBool=AND Boolean(A=@n8, B=@n7))) -> (Output_Get:double)
   n9 = Get bIsSliding()
   > Set OnlyCrouching(OnlyCrouching:bool=AND Boolean(A=Get bIsCrouching(), B=NOT Boolean(A=@n9))) -> (Output_Get:bool)
   n10 = Get bIsPlayingEmote()
   n11 = Get bIsSkyDiving()
   > Set FallingOnGround(FallingOnGround:bool=AND Boolean!(A=Get bIsFalling(), B=NOT Boolean(A=@n11), C:bool=NOT Boolean(A=@n10), D:bool=Get bIsMoving2D()) -> (ReturnValue)) -> (Output_Get:bool)
   n12 = Get bIsSkyDiving()
   n13 = Get bIsPlayingEmote()
   n14 = Get bIsMoving2D()
   > Set FallingOnGround_Idle(FallingOnGround_Idle:bool=AND Boolean(A=Get bIsFalling(), B=NOT Boolean(A=@n12), C=NOT Boolean(A=@n13), D=NOT Boolean(A=@n14))) -> (Output_Get:bool)
   > Set OnGrindRail RBAN(OnGrindRail RBAN:bool=AND Boolean!(A=Get bIsOnGrindRail(), B=Get bRigidBodyEnabled()) -> (ReturnValue)) -> (Output_Get:bool)
   > Set LOD 3(LOD 3:bool="integer > integer"(A=Get CurrentLOD(), B=2)) -> (Output_Get:bool)
   n15 = Player Movement Function DEPRECATED!(
    execute=@n1.then_6
    CharacterPartAnimInstance:Custom Character Part Anim Instance Object Reference=Self-Reference()
    X Dot Product:double=100
    Y Dot Product:double=100
    Z Dot Product:double=100
    DrawDebug:bool=false
    Debug Scalar:double=50
   ) -> (then, X Movement:double, Y Movement:double, Z Movement:double)
   n16 = "float < float"(A=BreakVector(InVec=MakeVector(X=@n15, Y=@n15.Y Movement)).Y, B=-40)
   n17 = Get OnlyCrouching()
   > Set Backward(Backward:double="To Float (Boolean)"(InBool=AND Boolean(A=@n16, B=@n17))) -> (Output_Get:double)
   n18 = Branch(execute=@n1.then_3, Condition=Get bIsOnGrindRail())
   > Set SpeedWaveJacket(SpeedWaveJacket:double=8) -> (Output_Get:double)
   > Set AmplitudWaveJacket(AmplitudWaveJacket:double=2) -> (Output_Get:double)
   > Set OffsetWaveJacket(OffsetWaveJacket:double=6) -> (Output_Get:double)
   n19 = Branch(execute=@n18.else, Condition=Get bIsBaseballSliding())
   > Set SpeedWaveJacket(SpeedWaveJacket=9)
   > Set AmplitudWaveJacket(AmplitudWaveJacket=1)
   > Set OffsetWaveJacket(OffsetWaveJacket=3)
   n20 = Branch(execute=@n19.else, Condition=Get bIsKneeSliding())
   > Set SpeedWaveJacket(SpeedWaveJacket=7)
   > Set AmplitudWaveJacket(AmplitudWaveJacket=1)
   > Set OffsetWaveJacket(OffsetWaveJacket=5)
   Set SpeedWaveJacket(execute=@n20.else, SpeedWaveJacket=0)
   > Set AmplitudWaveJacket(AmplitudWaveJacket=0)
   > Set OffsetWaveJacket(OffsetWaveJacket=0)
   n21 = Branch(execute=@n1.then_1, Condition=Get bIsOnGrindRail())
   > Set Alpha RB Default(Alpha RB Default:double=0.7) -> (Output_Get:double)
   n22 = Branch(execute=@n21.else, Condition=AND Boolean(A=Get bIsJumping(), B=Get Is Wearing Cape()))
   > Set Alpha RB Default(Alpha RB Default=0.5)
   Set Alpha RB Default(execute=@n22.else, Alpha RB Default=1)
   n23 = Branch(execute=@n1.then_7, Condition=Get bRigidBodyEnabled())
   > SetSkinWeightProfile!(
    execute
    self:Skinned Mesh Component Object Reference=Get Body Skel Part() -> (Body Skel Part:Skeletal Mesh Component Object Reference)
    InProfileName:Name=None
    InLayer:ESkinWeightProfileLayer Enum=Primary
   ) -> (then, ReturnValue:bool)
   n24 = Branch(execute=@n23.else, Condition=Get LOD 3() -> (LOD 3:bool))
   > SetSkinWeightProfile(self=Get Body Skel Part(), InProfileName=AlternativeLOD3, InLayer=Primary)
   SetSkinWeightProfile(execute=@n24.else, self=Get Body Skel Part(), InProfileName=Alternative, InLayer=Primary)
   n25 = float * float(A=GetRealTimeSeconds(), B=Get SpeedWaveJacket() -> (SpeedWaveJacket:double)) {is=title K2Node_PromotableOperator}
   n26 = float * float(A=MultiplyByPi!(Value:double="Sin (Radians)"(A=@n25)) -> (ReturnValue:double), B=Get AmplitudWaveJacket() -> (AmplitudWaveJacket:double))
   n27 = Get OffsetWaveJacket() -> (OffsetWaveJacket:double)
   Set RotatorWaveJacket(execute=@n1.then_4, RotatorWaveJacket:Rotator=MakeRotator(Pitch=float + float(A=@n26, B=@n27))) -> (Output_Get:Rotator)
   n28 = GetAnimInstance(self=GetOwnerSkeletalMeshComponent())
   > Set PlayinhEasternBloc(PlayinhEasternBloc:bool=Montage_IsPlaying(self=@n28, Montage=/BRCosmetics/Animation/Game/MainPlayer/Emotes/Eastern_Bloc/Emote_EasternBloc_CMM_M)) -> (Output_Get:bool)
   > Set PlayinhBreakDance(PlayinhBreakDance:bool=Montage_IsPlaying(self=@n28, Montage=/BRCosmetics/Animation/Game/MainPlayer/Emotes/Emote_Breakdance_M1)) -> (Output_Get:bool)
   > Set PlayingHeadBanger(PlayingHeadBanger:bool=Montage_IsPlaying(self=@n28, Montage=/BRCosmetics/Animation/Game/MainPlayer/Emotes/Headbanger/Emote_Headbanger_CMM_M)) -> (Output_Get:bool)
   > n29 = Branch(Condition=Montage_IsPlaying(self=@n28, Montage=/BRCosmetics/Animation/Game/MainPlayer/Emotes/BalletSpin/Emote_BalletSpin_CMM_M))
   > Delay(Duration=1)
   > Set PlayingBalletSpin(PlayingBalletSpin:double=1) -> (Output_Get:double)
   Set PlayingBalletSpin(execute=@n29.else, PlayingBalletSpin=0)
   Event OnCharacterCustomizationLoaded()
   > Set Body Skel Part(Body Skel Part:Skeletal Mesh Component Object Reference=GetPartSkeletalMeshComponent()) -> (Output_Get:Skeletal Mesh Component Object Reference)
   Event OnWeaponChanged()
   > n30 = Cast To FortPlayerAnimInstance(Object=GetAnimInstance(self=GetOwnerSkeletalMeshComponent()))
   > Set Detected_Pistol(Detected_Pistol:double=0) -> (Output_Get:double)
   > n31 = Switch on EFortWeaponCoreAnimation!(execute, Selection:EFortWeaponCoreAnimation Enum=Get WeaponCoreAnim(self=@n30)) -> (
    Melee:Exec
    Pistol:Exec
    Shotgun:Exec
    PaperBlueprint:Exec
    Rifle:Exec
    MeleeOneHand:Exec
    MachinePistol:Exec
    RocketLauncher:Exec
    GrenadeLauncher:Exec
    GoingCommando:Exec
    AssaultRifle:Exec
    TacticalShotgun:Exec
    SniperRifle:Exec
    TrapPlacement:Exec
    ShoulderLauncher:Exec
    AbilityDecoTool:Exec
    Crossbow:Exec
    C4:Exec
    RemoteControl:Exec
    DualWield:Exec
    AR_BullPup:Exec
    AR_ForwardGrip:Exec
    MedPackPaddles:Exec
    SMG_P90:Exec
    AR_DrumGun:Exec
    Consumable_Small:Exec
    Consumable_Large:Exec
    Balloon:Exec
    MountedTurret:Exec
    CreativeTool:Exec
    ExplosiveBow:Exec
    AshtonIndigo:Exec
    AshtonChicago:Exec
    MeleeDualWield:Exec
    Unarmed:Exec
   )
   Set Detected_Pistol(execute=@n31.Pistol, Detected_Pistol=1)
  ]}
]
variables [
 Alpha RB Default {type=number}
 Alpha_SkydiveFE And Emote {type=number}
 AmplitudWaveJacket {type=number}
 Backward {type=number}
 Body Skel Part {type=SkeletalMeshComponent}
 CrouchingIdle {type=number}
 Detected_Pistol {type=number}
 FallingOnGround {type=boolean}
 FallingOnGround_Idle {type=boolean}
 Gravity Override {type=Vector}
 Is Wearing Cape {type=boolean}
 JumpingWithCape {type=number}
 LOD 3 {type=boolean}
 OffsetWaveJacket {type=number}
 OnGrindRail RBAN {type=boolean}
 OnlyCrouching {type=boolean}
 PlayingBalletSpin {type=number}
 PlayingHeadBanger {type=boolean}
 PlayinhBreakDance {type=boolean}
 PlayinhEasternBloc {type=boolean}
 RBAN Settings Out {type=FortRigidBodyAnimNodeParameters}
 RotatorWaveJacket {type=Rotator}
 SpeedWaveJacket {type=number}
 Test {type=number}
]
