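# Physics asset of cape colliders for the M_MED_PeaceScarf male medium player body, with no constraints between bodies.
# Capsules cover the pelvis, spine, neck and clavicles, and a sphere covers the head; arm, leg and foot bodies have no shapes.
# Three dyn_jacket_bk chains (left, center, right) carry capsules on four bones each, with the last bone stacking three capsules to extend coverage along the bone.
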PhysicsAsset /BRCosmetics/Characters/Player/Male/Medium/Bodies/M_MED_PeaceScarf/Meshes/M_MED_PeaceScarf_CapeColliders
bodies {
 calf_l []
 calf_r []
 clavicle_l [{shapeType=Capsule, length=9.0192, radius=9.8871, center=<9.4266 -1.1806 -2.4977>, rotation=<rot 69.3297 1.4476 176.9462>}]
 clavicle_r [{shapeType=Capsule, length=9.0192, radius=9.8871, center=<-9.4266 1.1806 2.4977>, rotation=<rot 69.3297 1.4476 -3.0538>}]
 dyn_jacket_bk_01_c [{shapeType=Capsule, length=12, radius=8, shapeName=dyn_jacket_bk_01_c_capsule, center=<5.4944 0 0>, rotation=<rot 90 -90 0>}]
 dyn_jacket_bk_01_l [{shapeType=Capsule, length=12, radius=8, shapeName=dyn_jacket_bk_01_l_capsule, center=<5.4944 0 0>, rotation=<rot 90 -90 0>}]
 dyn_jacket_bk_01_r [{shapeType=Capsule, length=12, radius=8, shapeName=dyn_jacket_bk_01_r_capsule, center=<5.4944 0 0>, rotation=<rot 90 -90 0>}]
 dyn_jacket_bk_02_c [{shapeType=Capsule, length=16.4122, radius=8, shapeName=dyn_jacket_bk_02_c_capsule, center=<5.4944 0 0.1508>, rotation=<rot 90 -90 0>}]
 dyn_jacket_bk_02_l [{shapeType=Capsule, length=16.4122, radius=8, shapeName=dyn_jacket_bk_02_l_capsule, center=<5.4944 0 0.1508>, rotation=<rot 90 -90 0>}]
 dyn_jacket_bk_02_r [{shapeType=Capsule, length=16.4122, radius=8, shapeName=dyn_jacket_bk_02_r_capsule, center=<5.4944 0 0.1508>, rotation=<rot 90 -90 0>}]
 dyn_jacket_bk_03_c [{shapeType=Capsule, length=16.3989, radius=8, shapeName=dyn_jacket_bk_03_c_capsule, center=<2.4504 0 1.6579>, rotation=<rot 90 -90 0>}]
 dyn_jacket_bk_03_l [{shapeType=Capsule, length=16.3989, radius=8, shapeName=dyn_jacket_bk_03_l_capsule, center=<2.4504 0 1.6579>, rotation=<rot 90 -90 0>}]
 dyn_jacket_bk_03_r [{shapeType=Capsule, length=16.3989, radius=8, shapeName=dyn_jacket_bk_03_r_capsule, center=<2.4504 0 1.6579>, rotation=<rot 90 -90 0>}]
 dyn_jacket_bk_04_c [
  {shapeType=Capsule, length=19.2638, radius=8, shapeName=dyn_jacket_bk_04_c_capsule, center=<1.3661 0 2.0586>, rotation=<rot 90 -90 0>}
  {shapeType=Capsule, length=19, radius=9, shapeName=dyn_jacket_bk_04_c_capsule, center=<13.861 0 0>, rotation=<rot 90 -90 0>}
  {shapeType=Capsule, length=19, radius=11, shapeName=dyn_jacket_bk_04_c_capsule, center=<28.5525 0 4.3179>, rotation=<rot 90 -90 0>}
 ]
 dyn_jacket_bk_04_l [
  {shapeType=Capsule, length=19.2638, radius=8, shapeName=dyn_jacket_bk_04_l_capsule, center=<1.7871 0 0.6949>, rotation=<rot 90 -90 0>}
  {shapeType=Capsule, length=18.5, radius=10, shapeName=dyn_jacket_bk_04_l_capsule, center=<14.8463 0 0>, rotation=<rot 90 -90 0>}
  {shapeType=Capsule, length=19.5, radius=10.5, shapeName=dyn_jacket_bk_04_l_capsule, center=<27.8966 0 7.7431>, rotation=<rot 90 -90 0>}
 ]
 dyn_jacket_bk_04_r [
  {shapeType=Capsule, length=19.2638, radius=8, shapeName=dyn_jacket_bk_04_r_capsule, center=<0.8035 0 0.6949>, rotation=<rot 90 -90 0>}
  {shapeType=Capsule, length=20.5, radius=9.5, shapeName=dyn_jacket_bk_04_r_capsule, center=<14 -5.8432 6.5098>, rotation=<rot 90 -90 0>}
  {shapeType=Capsule, length=19.5, radius=10.5, shapeName=dyn_jacket_bk_04_r_capsule, center=<28.0756 -6.204 9.0397>, rotation=<rot 90 -90 0>}
 ]
 foot_l []
 foot_r []
 head [{shapeType=Sphere, radius=17.3125, center=<11.028 -0.174 0>}]
 lowerarm_l []
 lowerarm_r []
 neck_02 [{shapeType=Capsule, length=13.4375, radius=7.3175, center=<-0.2642 -0.3408 0>, rotation=<rot 90 -171.1193 -175.3978>}]
 pelvis [{shapeType=Capsule, length=8.6456, radius=14.5388, center=<-2.4887 -3.0615 0>}]
 spine_01 [{shapeType=Capsule, length=7.9438, radius=13.5586, center=<-1.7965 -3.0321 0>}]
 spine_02 [
  {shapeType=Capsule, length=7.3264, radius=13.6322, center=<-1.711 -3.4049 0.0001>}
  {shapeType=Capsule, length=29.3956, radius=11.5807, center=<2.5024 0.4992 3.9825>, rotation=<rot 90 0.0049 0.0049>}
  {shapeType=Capsule, length=24.0419, radius=13.2025, center=<2.4704 -3.9205 -3.9825>, rotation=<rot 90 5.1733 0>}
 ]
 spine_03 [{shapeType=Capsule, length=8.2107, radius=13.7031, center=<0.1445 0.2396 0.0001>}]
 spine_04 [{shapeType=Capsule, length=9.4404, radius=14.1325, center=<2.766 0.7765 0.0002>}]
 spine_05 [
  {shapeType=Capsule, length=8.0226, radius=15.695, center=<-2.2795 1.8109 0>}
  {shapeType=Capsule, length=13.9518, radius=13.9425, center=<-0.2916 0.4452 0>}
 ]
 thigh_l []
 thigh_r []
 upperarm_l []
 upperarm_r []
}
constraints []
